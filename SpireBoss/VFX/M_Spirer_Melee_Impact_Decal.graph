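# Deferred decal material for the Spirer boss's melee impact, marking the ground with charred dirt and a fading emissive glow.
# An alpha mask shapes opacity alongside lifetime and facing fades, and blends tiled dirt with a charred color using dirt-alpha height.
# Glow comes from that mask or an optional crack texture, colored by center-scaling turbulence inside a sphere mask as it fades.
# Mobile shading paths sample textures at explicit mip levels.
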Material /SpireBoss/VFX/M_Spirer_Melee_Impact_Decal
graph {
 properties {blendMode=BLEND_Translucent, materialDomain=MD_DeferredDecal, pixelDepthOffsetMode=PDOM_Legacy, usedShadingModels=Default Lit}
 nodes [
  n1 = Multiply(
   A=TextureCoordinate()
   B=ComponentMask(None=VectorParameter() {parameterName=Ground UV Scale, defaultValue=<rgba 3 3 0 0>}.RGB) {g=true, r=true}
  )
  n2 = TextureSampleParameter2D(UVs=@n1) {
   parameterName=Ground_Diffuse
   samplerSource=SSM_Wrap_WorldGroupSettings
   texture=/Game/Environments/Apollo/Sets/Terrain/Textures/T_Apollo_Dirt_D_02
  }
  n3 = TextureSampleParameter2D!(UVs=@n1, Level=Constant(), Apply View MipBias) {
   mipValueMode=TMVM_MipLevel
   parameterName=Ground_Diffuse
   samplerSource=SSM_Wrap_WorldGroupSettings
   texture=&T_Apollo_Dirt_D_02
  }
  n4 = TextureSampleParameter2D!(UVs, Apply View MipBias) {
   parameterName=Alpha Mask
   samplerSource=SSM_Wrap_WorldGroupSettings
   texture=/Game/Packages/Fortress_Transylvania/Haunted_Forest/Swamps/Textures/T_dust_cloud_01
  }
  n5 = TextureSampleParameter2D!(UVs, Level=Constant(), Apply View MipBias) {
   mipValueMode=TMVM_MipLevel
   parameterName=Alpha Mask
   samplerSource=SSM_Wrap_WorldGroupSettings
   texture=&T_dust_cloud_01
  }
  n6 = ShadingPathSwitch(Default=@n4.R, Mobile=@n5.R)
  n7 = MaterialFunctionCall(In=@n6, Contrast=ScalarParameter() {defaultValue=1, parameterName=Contrast}) {materialFunction=&CheapContrast}
  n8 = OneMinus(None=Power(Base=ShadingPathSwitch(Default=@n2.A, Mobile=@n3.A)) {constExponent=5})
  LinearInterpolate(
   A=VectorParameter() {parameterName=Charred Color}.RGB
   B=ShadingPathSwitch(Default=@n2.RGB, Mobile=@n3.RGB)
   Alpha=Multiply(A=Subtract(A=@n6, B=@n7.Result), B=@n8)
  ) {constB=0}
  n9 = TextureSampleParameter2D!(UVs=@n1, Apply View MipBias) {
   parameterName=Ground_SpecularMap
   samplerSource=SSM_Wrap_WorldGroupSettings
   samplerType=SAMPLERTYPE_Masks
   texture=/Game/Environments/Apollo/Sets/Terrain/Textures/T_Apollo_Dirt_S_02
  }
  n10 = TextureSampleParameter2D!(UVs=@n1, Level=Constant(), Apply View MipBias) {
   mipValueMode=TMVM_MipLevel
   parameterName=Ground_SpecularMap
   samplerSource=SSM_Wrap_WorldGroupSettings
   samplerType=SAMPLERTYPE_Masks
   texture=&T_Apollo_Dirt_S_02
  }
  MaterialFunctionCall(Float3=ShadingPathSwitch(Default=@n9.RGB, Mobile=@n10.RGB)) {materialFunction=&BreakOutFloat3Components}
  n11 = DecalLifetimeOpacity()
  Multiply(
   A=Multiply(A=@n6, B=@n11.Opacity)
   B=MaterialFunctionCall() {materialFunction=/Game/Weapons/Prototype/SprayGun/Materials/MF_DecalFacingMask}.Result
  )
  n12 = Multiply(
   A=TextureCoordinate()
   B=AppendVector(
    A=ScalarParameter() {defaultValue=1, group=Emissive, parameterName=Emissive U Scale, sortPriority=2}
    B=ScalarParameter() {defaultValue=1, group=Emissive, parameterName=Emissive V Scale, sortPriority=2}
   )
  )
  n13 = TextureSampleParameter2D!(UVs=@n12, Apply View MipBias) {
   group=Emissive
   parameterName=Emissive Texture
   samplerSource=SSM_Wrap_WorldGroupSettings
   samplerType=SAMPLERTYPE_LinearColor
   texture=/Game/Global/Textures/Default/Masks_and_Gradients/T_Cracks
  }
  n14 = TextureSampleParameter2D!(UVs=@n12, Level=Constant(), Apply View MipBias) {
   group=Emissive
   mipValueMode=TMVM_MipLevel
   parameterName=Emissive Texture
   samplerSource=SSM_Wrap_WorldGroupSettings
   samplerType=SAMPLERTYPE_LinearColor
   texture=&T_Cracks
  }
  n15 = ShadingPathSwitch(
   Default=StaticComponentMaskParameter(None=@n13.RGBA) {defaultR=true, group=Emissive, parameterName=Emissive Mask, sortPriority=1}
   Mobile=StaticComponentMaskParameter(None=@n14.RGBA) {defaultR=true, group=Emissive, parameterName=Emissive Mask, sortPriority=1}
  ) {desc=Migrated from FeatureLevelSwitch}
  n16 = ScalarParameter() {parameterName=Use Unique Emissive Mask}
  n17 = Clamp(None=Power(Base=@n8) {constExponent=10})
  n18 = ScalarParameter() {defaultValue=1, parameterName=Emissive Mask Intensity}
  n19 = ScalarParameter() {defaultValue=3, group=Emissive, parameterName=Emissive Fade Curve Control, sortPriority=6}
  n20 = VectorParameter() {group=Emissive, parameterName=Emissive Color A, defaultValue=<rgba 0.0097 0.0015 0.0208 1>}
  n21 = VectorParameter() {group=Emissive, parameterName=Emissive Color B, defaultValue=<rgba 6.619 0.52 10 1>}
  n22 = LinearInterpolate(Alpha=OneMinus(None=DecalLifetimeOpacity().Opacity)) {constA=1, constB=3}
  n23 = Multiply(
   A=TextureSample(UVs=MaterialFunctionCall(Texture Scale=@n22) {materialFunction=&ScaleUVsByCenter}.UVs) {texture=/Game/Global/Textures/Default/Caustics/T_Turbulence_Seamless}.G
   B=SphereMask(
    A=TextureCoordinate()
    B=Constant() {r=0.5}
    Radius=ScalarParameter() {defaultValue=0.5, group=Emissive, parameterName=Emissive Fade Radius, sortPriority=3}
   ) {hardnessPercent=0}
  )
  Multiply(
   A=Multiply(A=LinearInterpolate(A=@n6, B=@n15, Alpha=@n16), B=Multiply(A=@n17, B=@n18))
   B=Multiply(A=Power(Base=@n11.Opacity, Exp=@n19) {constExponent=3}, B=LinearInterpolate(A=@n20.RGB, B=@n21.RGB, Alpha=@n23))
  )
  n24 = TextureSampleParameter2D!(UVs=@n1, Apply View MipBias) {
   parameterName=Ground_Normal
   samplerSource=SSM_Wrap_WorldGroupSettings
   samplerType=SAMPLERTYPE_Normal
   texture=/Game/Environments/Apollo/Sets/Terrain/Textures/T_Apollo_Dirt_N_02
  }
  n25 = TextureSampleParameter2D!(UVs=@n1, Level=Constant(), Apply View MipBias) {
   mipValueMode=TMVM_MipLevel
   parameterName=Ground_Normal
   samplerSource=SSM_Wrap_WorldGroupSettings
   samplerType=SAMPLERTYPE_Normal
   texture=&T_Apollo_Dirt_N_02
  }
  Reroute(None=ShadingPathSwitch(Default=@n24.RGB, Mobile=@n25.RGB))
 ]
}
