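# Translucent lit material for Niagara mesh particles, named as the spire boss's fire mask projectile.
# Panning, self-distorted flame noise builds masks, with camera-facing Fresnel terms eroded by noise for opacity and color.
# Three tint colors are blended and intensity-scaled, and a pre-skinned local position yields a vertical gradient mask.
# Scrolling soft noise along vertex normals, scaled between minimum and maximum strengths, forms a world position offset term.
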Material /SpireBoss/VFX/Materials/M_FireMask_Projectile
graph {
 properties {bAutomaticallySetUsageInEditor=false, bUsedWithNiagaraMeshParticles=true, blendMode=BLEND_Translucent, usedShadingModels=Default Lit}
 nodes [
  n1 = Time()
  n2 = VectorParameter() {parameterName=TextureNoisePanSpeedAndDirection, defaultValue=<rgba 0 0 1 1>}
  MaterialFunctionCall(
   TextureObject=TextureObject() {texture=/Game/Effects/FogVolume/noise_flamy_Tile}
   TextureSize=VectorParameter() {parameterName=TextureNoiseSize, defaultValue=<rgba 32 32 32 1>}.RGB
   WorldPosition=Add(A=NamedRerouteUsage(), B=Multiply(A=@n1, B=@n2.RGB))
  ) {materialFunction=&WorldAlignedTexture}
  n3 = TextureCoordinate()
  n4 = TextureSample(UVs=Panner(Coordinate=Multiply(A=@n3) {constB=3}) {speedY=1}) {texture=/Game/Characters/Player/Male/Medium/Bodies/M_MED_Viper/Textures/Glitch/T_SoftNoise}
  n5 = NamedRerouteUsage()
  Multiply(
   A=Multiply(A=VertexNormalWS(), B=@n4.RGB) {constB=2}
   B=LinearInterpolate(A=ScalarParameter() {defaultValue=0.1, parameterName=MinWpo}, B=ScalarParameter() {defaultValue=2, parameterName=MaxWPO}, Alpha=@n5)
  )
  Multiply(A=@n4.RGB, B=@n5)
  n6 = TextureCoordinate()
  n7 = ScalarParameter() {defaultValue=1, parameterName=Debug_Tile}
  n8 = Panner(Coordinate=Multiply(A=TextureCoordinate(), B=ScalarParameter() {defaultValue=2, parameterName=Debug_Tile2})) {speedY=0.25}
  n9 = Subtract(A=Multiply(A=TextureSample(UVs=@n8) {texture=&noise_flamy_Tile}.RGB) {constB=2})
  n10 = ScalarParameter() {defaultValue=0.1, parameterName=Distortion_Strength}
  n11 = Add(A=Panner(Coordinate=Multiply(A=@n6, B=@n7)) {speedY=0.1}, B=ComponentMask(None=Multiply(A=@n9, B=@n10)) {g=true, r=true})
  NamedRerouteDeclaration(None=ComponentMask(None=TextureSample(UVs=@n11) {texture=&noise_flamy_Tile}.RGB) {r=true}) {name=TextureNoise, nodeColor=<rgba 0.3828 1 0 1>}
  n12 = DotProduct(A=CameraVectorWS(), B=VertexNormalWS())
  n13 = ScalarParameter() {defaultValue=1, parameterName=OpacityFresnelPower}
  n14 = Multiply(A=NamedRerouteUsage(), B=ScalarParameter() {defaultValue=1, parameterName=OpacityNoise_Strength})
  n15 = Subtract(A=Saturate(None=Power(Base=@n12, Exp=@n13)), B=@n14)
  NamedRerouteDeclaration(None=Saturate(None=Multiply(A=@n15) {constB=2})) {name=Opacity, nodeColor=<rgba 0 0.625 1 1>}
  n16 = Panner(Coordinate=Multiply(A=TextureCoordinate(), B=ScalarParameter() {defaultValue=3, parameterName=OpacityNoise_Tile})) {speedY=0.2}
  n17 = Multiply(A=TextureCoordinate(), B=ScalarParameter() {defaultValue=4, parameterName=Opacity_Distort_Tile})
  n18 = Multiply(A=TextureSample(UVs=Panner(Coordinate=@n17) {speedY=0.5}) {texture=&noise_flamy_Tile}.RGB) {constB=2}
  n19 = ComponentMask(None=Multiply(A=Subtract(A=@n18), B=ScalarParameter() {defaultValue=0.1, parameterName=Distortion_Strength})) {g=true, r=true}
  NamedRerouteDeclaration(None=TextureSample(UVs=Add(A=@n16, B=@n19)) {texture=&noise_flamy_Tile}.R) {name=Opacity_Noise, nodeColor=<rgba 1 0.1641 0 1>}
  n20 = Power(Base=@n12, Exp=ScalarParameter() {defaultValue=3, parameterName=ColorFresnelPower})
  NamedRerouteDeclaration(None=Subtract(A=Saturate(None=@n20), B=@n14)) {name=ColorFresnel, nodeColor=<rgba 1 0.9609 0 1>}
  n21 = OneMinus(None=@n12)
  n22 = ScalarParameter() {defaultValue=4, parameterName=WPO_FresnelPower}
  NamedRerouteDeclaration(None=Saturate(None=Power(Base=@n21, Exp=@n22))) {name=WPO_Fresnel, nodeColor=<rgba 0 0.6719 1 1>}
  n23 = LocalPosition() {localOrigin=InstancePreSkinning}
  NamedRerouteDeclaration(None=ShaderStageSwitch(PixelShader=VertexInterpolator(VS=@n23.XYZ).PS, VertexShader=@n23.XYZ)) {name=Preskinned_WorldPosition, nodeColor=<rgba 0 1 0.6016 1>}
  n24 = Divide(
   A=Add(A=NamedRerouteUsage(), B=VectorParameter() {parameterName=PreskinnedPosition_Offset, defaultValue=<rgba 0 0 16 1>}.RGB)
   B=ScalarParameter() {defaultValue=42, parameterName=Divide}
  )
  NamedRerouteDeclaration(None=ComponentMask(None=Saturate(None=@n24)) {b=true}) {name=Z-Gradient, nodeColor=<rgba 1 0.6094 0 1>}
  NamedRerouteUsage()
  NamedRerouteUsage()
  NamedRerouteUsage()
  PixelNormalWS()
  n25 = LinearInterpolate(
   A=LinearInterpolate(
    A=VectorParameter() {parameterName=Color_02, defaultValue=<rgba 1 0 0.25 1>}.RGB
    B=VectorParameter() {parameterName=Color_03, defaultValue=<rgba 0.1285 0.0033 0 1>}.RGB
    Alpha=NamedRerouteUsage()
   )
   B=VectorParameter() {parameterName=Color_01, defaultValue=<rgba 0.0041 0 0.0243 1>}.RGB
   Alpha=NamedRerouteUsage()
  )
  n26 = LinearInterpolate(Alpha=NamedRerouteUsage()) {constA=0.1, constB=2}
  Power(Base=Multiply(A=Multiply(A=@n25, B=@n26)) {constB=2})
  PreSkinnedLocalBounds()
  Subtract(A=Multiply(A=@n4.RGB) {constB=2})
  TextureSample() {samplerType=SAMPLERTYPE_Normal, texture=/DemonMaskFire/Assets/Textures/T_Mask_Fire_N}
 ]
}
