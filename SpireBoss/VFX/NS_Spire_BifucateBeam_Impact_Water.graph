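# Niagara impact effect for where the Spire boss's bifurcate beam hits water.
# Short bursts spawn mesh splashes, chevrons and foam, plus sprite drops and mist, around User Stop Location.
# Most emitters launch along a reflection vector built from Stop Normal; drops instead use point force and curl noise, losing lifetime faster with distance.
# Foam and drop materials receive water body index, depth and wave attenuation, and the main foam is constrained to the impact plane.
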NiagaraSystem /SpireBoss/VFX/NS_Spire_BifucateBeam_Impact_Water
dependencies {
 usedDataInterfaces [
  &NiagaraDataInterfaceCurve
  &NiagaraDataInterfaceVectorCurve
  &NiagaraDataInterfaceColorCurve
  &NiagaraDataInterfaceCurlNoise
 ]
 usedDynamicInputs []
 usedModules [&SystemState]
}
emitters {
 Arcing_Water_Splashes {
  emitterUpdateScript [
   &EmitterState(Life Cycle Mode=::NewEnumerator1, Loop Duration=0.1, UseLoopDelay=false)
   &SpawnBurst_Instantaneous(
    Spawn Count=&RandomRangeInt(Minimum=0, Maximum=1)
    Use Spawn Probability=false
    UseLoopCountLimit=false
   )
  ]
  particleSpawnScript [
   &InitializeParticle(
    Color Maximum=<rgba 1 1 1 0.75>
    Color Minimum=<rgba 1 1 1 0.25>
    Color Mode=::NewEnumerator2
    Lifetime Max=1.2
    Lifetime Min=0.8
    Lifetime Mode=::NewEnumerator1
    Mass Max=1.2
    Mass Min=0.6
    Mass Mode=::NewEnumerator3
    Mesh Scale Mode=::NewEnumerator1
    UsePositionOffset=false
    Write Material Random=true
   )
   &SphereLocation(
    MaskSpawnGroup=false
    Override Seed=false
    Sphere Coordinate Space=::World
    Sphere Origin:NiagaraPosition=$User.Stop Location
    Sphere Radius=20
    Surface Only=false
    UseNonUniformScale=false
    UseOffset=false
    UseSphereOrientation=false
   )
   &AddVelocityInCone(
    Cone Angle=60
    Cone Axis=&self:Reflection_Vector(
     Normalized Normal:Vector3f=&NormalizeVector(Vector To Normalize=$User.Stop Normal, Safe Normalize=true, Safe Fallback Threshold=0, Safe Fallback Vector=<0 0 0.001>)
     Stop Location:Vector3f=Linked() {linkedVariable {name=User.Stop Location, type=NiagaraPosition}}
    )
    Cone Axis Coordinate Space=::World
    Module.Use Velocity Falloff On Cone Axis=false
    Velocity Strength=50
   )
  ]
  particleUpdateScript [
   &Drag(Drag=2, Use Linear Drag=true, Use Rotational Drag=false)
   &SolveForcesAndVelocity(Clamp Velocity=false, Limit Acceleration=false)
   &ParticleState()
   &DynamicMaterialParameters(
    Index 0 Param 1=&FloatFromCurve(FloatCurve=DataInterface() {propertyValues {Curve {keys [
         {arriveTangent=0.8954, interpMode=RCIM_Cubic, leaveTangent=0.8954, tangentMode=RCTM_User, time=-0.0005, value=-1.845}
         {arriveTangent=1.8419, interpMode=RCIM_Cubic, leaveTangent=1.8419, time=0.7887, value=-0.7711}
         {arriveTangent=3.1029, interpMode=RCIM_Cubic, leaveTangent=3.1029, tangentMode=RCTM_User, time=1.0011, value=0}
        ]}}})
    Index 0 Param 2=&FloatFromCurve(FloatCurve=DataInterface() {propertyValues {Curve {keys [<at -0.0046 4>,<at 1.0011 0>]}}})
    Index 0 Param 3=&FloatFromCurve(
     FloatCurve=DataInterface() {propertyValues {Curve {keys [
         {arriveTangent=0.096, interpMode=RCIM_Cubic, leaveTangent=0.096, tangentMode=RCTM_User, time=-0.0046, value=0.0882}
         {arriveTangentWeight=0.4536, interpMode=RCIM_Cubic, tangentMode=RCTM_User, time=1.0011, value=0.175}
        ]}}}
     Scale Curve=3
    )
    Index 0 Param 4=&FloatFromCurve(FloatCurve=DataInterface() {propertyValues {Curve {keys [
         {
          arriveTangent=18.6242
          interpMode=RCIM_Cubic
          leaveTangent=18.6242
          leaveTangentWeight=2.1719
          tangentMode=RCTM_User
          time=-0.0046
          value=0.9276
         }
         {arriveTangentWeight=1.5889, interpMode=RCIM_Cubic, tangentMode=RCTM_User, time=1.0011, value=5.5296}
        ]}}})
    Module.Param0WriteEnabled=true
    Module.Param1WriteEnabled=true
    Module.Param2WriteEnabled=true
    Module.Param3WriteEnabled=true
   )
   &ScaleMeshSize(Scale Factor=&VectorFromCurve(
     VectorCurve=DataInterface() {propertyValues {
       XCurve {keys [{arriveTangent=3.3001, interpMode=RCIM_Cubic, leaveTangent=3.3001, tangentMode=RCTM_User, time=0, value=0.5}, <at 1 1.5 RCIM_Cubic>]}
       YCurve {keys [{arriveTangent=3.2853, interpMode=RCIM_Cubic, leaveTangent=3.2853, tangentMode=RCTM_User, time=0, value=0.5}, <at 0.9988 1.5 RCIM_Cubic>]}
       ZCurve {keys [
         {arriveTangent=0.9572, interpMode=RCIM_Cubic, leaveTangent=0.9572, tangentMode=RCTM_User, time=0, value=0.5}
         {arriveTangentWeight=0.618, interpMode=RCIM_Cubic, tangentMode=RCTM_User, time=0.9988, value=1}
        ]}
      }}
     Scale Curve=<1 1 0.75>
    ))
   &ScaleColor(
    Scale Alpha=&FloatFromCurve(FloatCurve=DataInterface() {propertyValues {Curve {keys [<at -0.0046 1.0113 RCIM_Cubic>,<at 1.0011 0.002 RCIM_Cubic>]}}})
    Scale RGB=&MakeVectorFromLinearColorRGB(Color=&ColorFromCurve(ColorCurve=DataInterface() {propertyValues {
        AlphaCurve {keys []}
        BlueCurve {keys [{arriveTangent=-26.4823, interpMode=RCIM_Cubic, leaveTangent=-26.4823, tangentMode=RCTM_User, time=0.1, value=10}, <at 1 1 RCIM_Cubic>]}
        GreenCurve {keys [{arriveTangent=3.1252, interpMode=RCIM_Cubic, leaveTangent=3.1252, tangentMode=RCTM_User, time=0.1, value=0.3013}, <at 1 1 RCIM_Cubic>]}
        RedCurve {keys [{arriveTangent=-10.3723, interpMode=RCIM_Cubic, leaveTangent=-10.3723, tangentMode=RCTM_User, time=0.1, value=4.989}, <at 1 1 RCIM_Cubic>]}
       }}))
    ScaleA=true
    ScaleRGB=true
   )
  ]
  rendererClasses [&NiagaraMeshRendererProperties]
  renderers [{rendererClass=&NiagaraMeshRendererProperties, rendererIndex=0}]
 }
 ChevronShapes {
  emitterUpdateScript [
   &EmitterState(Life Cycle Mode=::NewEnumerator1, Loop Duration=0.15, UseLoopDelay=false)
   &SpawnBurst_Instantaneous(Use Spawn Probability=false, UseLoopCountLimit=false)
  ]
  particleSpawnScript [
   &InitializeParticle(
    Color=<rgba 1 1 1 0.1>
    Lifetime Max=0.8
    Lifetime Min=0.5
    Lifetime Mode=::NewEnumerator1
    Mesh Scale=<2 2 10>
    Mesh Scale Mode=::NewEnumerator4
    UsePositionOffset=false
    Write Material Random=true
   )
   &SystemLocation(
    MaskSpawnGroup=false
    Offset=/Niagara/DynamicInputs/Transforms/TransformPosition(
     Position=&Add_Vector(A=Linked() {linkedVariable {name=User.Stop Location, type=NiagaraPosition}}, B=<0 0 0>)
     Source Space=::World
     Destination Space=::Local
    )
   )
   &AddVelocityInCone(
    Cone Angle=60
    Cone Axis=&self:Reflection_Vector(
     Normalized Normal:Vector3f=&NormalizeVector(Vector To Normalize=$User.Stop Normal, Safe Normalize=true, Safe Fallback Threshold=0, Safe Fallback Vector=<0 0 0.001>)
     Stop Location:Vector3f=Linked() {linkedVariable {name=User.Stop Location, type=NiagaraPosition}}
    )
    Cone Axis Coordinate Space=::World
    Module.Use Velocity Falloff On Cone Axis=false
    Velocity Strength=500
   )
   /SetVariables(Particles.Velocity=&Multiply_Vector(A=$Particles.Velocity, B=<1 1 0>)) {isSetParameters=true}
  ]
  particleUpdateScript [
   &SolveForcesAndVelocity(Clamp Velocity=false, Limit Acceleration=false)
   &ParticleState()
   &DynamicMaterialParameters(
    Index 0 Param 1=&Add_Float(
     A=&Add_Float(A=$Particles.MaterialRandom, B=1024)
     B=&FloatFromCurve(
      FloatCurve=DataInterface() {propertyValues {Curve {keys [
          {arriveTangent=-1.5599, interpMode=RCIM_Cubic, leaveTangent=-1.5599, tangentMode=RCTM_User, time=-0.0046, value=1.0113}
          <at 1.0011 0.002 RCIM_Cubic>
         ]}}}
      Scale Curve=$Particles.Initial.Lifetime
     )
    )
    Index 0 Param 2=&FloatFromCurve(FloatCurve=DataInterface() {propertyValues {Curve {keys [
         {arriveTangent=0.0979, interpMode=RCIM_Cubic, leaveTangent=0.0979, tangentMode=RCTM_User, time=0.2268, value=0.1}
         {interpMode=RCIM_Cubic, tangentMode=RCTM_User, time=1.0011, value=1.8}
        ]}}})
    Index 0 Param 4=1
    Module.Param0WriteEnabled=true
    Module.Param1WriteEnabled=true
    Module.Param2WriteEnabled=true
    Module.Param3WriteEnabled=true
   )
   &ScaleMeshSize(Scale Factor=&MakeVector(
     X=&FloatFromCurve(FloatCurve=DataInterface() {propertyValues {Curve {keys [
          {arriveTangent=3.3073, interpMode=RCIM_Cubic, leaveTangent=3.3073, leaveTangentWeight=1.2626, tangentMode=RCTM_User, time=0, value=2}
          <at 1.0011 4 RCIM_Cubic>
         ]}}})
     Y=&FloatFromCurve(FloatCurve=DataInterface() {propertyValues {Curve {keys [
          {arriveTangent=4.0016, interpMode=RCIM_Cubic, leaveTangent=4.0016, tangentMode=RCTM_User, time=-0.0046, value=1.0113}
          <at 1.0011 4 RCIM_Cubic>
         ]}}})
     Z=&FloatFromCurve(FloatCurve=DataInterface() {propertyValues {Curve {keys [
          {arriveTangent=9.2177, interpMode=RCIM_Cubic, leaveTangent=9.2177, tangentMode=RCTM_User, time=-0.0046, value=0}
          {
           arriveTangent=-2.7817
           arriveTangentWeight=0.1643
           interpMode=RCIM_Cubic
           leaveTangent=-2.7817
           leaveTangentWeight=0.3781
           tangentMode=RCTM_User
           time=0.3
           value=0.75
          }
          <at 1.0011 0.002 RCIM_Cubic>
         ]}}})
    ))
  ]
  rendererClasses [&NiagaraMeshRendererProperties]
  renderers [{rendererClass=&NiagaraMeshRendererProperties, rendererIndex=0}]
 }
 Drop_Details {
  emitterUpdateScript [
   &EmitterState(Life Cycle Mode=::NewEnumerator1, Loop Duration=0.25, UseLoopDelay=false)
   &SpawnBurst_Instantaneous(
    Spawn Count=&RandomRangeInt(Minimum=15, Maximum=45)
    Use Spawn Probability=false
    UseLoopCountLimit=false
   )
  ]
  particleSpawnScript [
   &InitializeParticle(
    Color=<rgba 1 1 1 0.75>
    Lifetime Max=3
    Lifetime Mode=::NewEnumerator1
    Sprite Rotation Mode=::NewEnumerator2
    Sprite Size Mode=::NewEnumerator1
    Uniform Sprite Size Max=6
    Uniform Sprite Size Min=1
    UsePositionOffset=false
    Write Material Random=true
   )
   /SetVariables(Particles.SpawnBeamLocation:NiagaraPosition=$User.Stop Location) {isSetParameters=true}
   &CylinderLocation(
    Cylinder Height=0
    Cylinder Height Midpoint=0
    Cylinder Origin:NiagaraPosition=$User.Stop Location
    Cylinder Radius=600
    MaskSpawnGroup=false
    Offset=<0 250 0>
    Offset Coordinate Space:ENiagaraCoordinateSpace=::Local
    Offset Mode:ENiagara_OffsetMode=::NewEnumerator0
    Override Seed=false
    UseOwnerScale=true
   )
   &DynamicMaterialParameters(
    Index 0 Param 1=1
    Index 3 Param 1=$User.WaterBodyIndex
    Index 3 Param 2=$User.WaterDepth
    Index 3 Param 3=$User.WaveAttenuationFactor
    Module.Param0WriteEnabled=true
    Module.Param0WriteEnabled3=true
    Module.Param1WriteEnabled=false
    Module.Param1WriteEnabled3=true
    Module.Param2WriteEnabled=false
    Module.Param2WriteEnabled3=true
    Module.Param3WriteEnabled=false
    Module.Param3WriteEnabled3=true
    Write Parameter Index 3=true
   )
   &SpriteFacingAndAlignment(Facing Coordinate Space=::World, Sprite Facing=<0 0 1>, WriteAlignment=false, WriteFacing=true)
  ]
  particleUpdateScript [
   &ParticleState()
   &PointForce(
    Force Falloff Distance=500
    Force Origin=$Particles.SpawnBeamLocation
    Force Strength=200
    Module.Use Falloff=false
    Module.UseLinearForce=false
   )
   /Niagara/Modules/Update/Forces/CurlNoiseForce(
    Noise Frequency=1
    Noise Strength=&FloatFromCurve(FloatCurve=DataInterface() {propertyValues {Curve {keys [
         {arriveTangent=-1159.6794, interpMode=RCIM_Cubic, leaveTangent=-1159.6792, tangentMode=RCTM_User, time=-0.0046, value=200}
         {arriveTangent=-174, interpMode=RCIM_Cubic, leaveTangent=-174, time=0.2223, value=70}
         <at 1.0011 25 RCIM_Cubic>
        ]}}})
    Pan Noise=false
   )
   &SolveForcesAndVelocity(Clamp Velocity=false, Limit Acceleration=false)
   /SetVariables(Particles.Lifetime=&Subtract_Float(
     A=$Particles.Lifetime
     B=&Multiply_Float(
      A=&Power(
       a=&VectorLength(Vector=&Subtract_Vector(A=$Particles.SpawnBeamLocation, B=Linked() {linkedVariable {name=Particles.Position, type=NiagaraPosition}}))
       b=1
      )
      B=&Multiply_Float(A=0.01, B=$Engine.DeltaTime)
     )
    )) {isSetParameters=true}
  ]
  rendererClasses [&NiagaraSpriteRendererProperties]
  renderers [{rendererClass=&NiagaraSpriteRendererProperties}]
 }
 Foam_Details {
  emitterUpdateScript [
   &EmitterState(Life Cycle Mode=::NewEnumerator1, Loop Delay=0.05, Loop Duration=0.25, UseLoopDelay=false)
   &SpawnBurst_Instantaneous(
    Spawn Count=&RandomRangeInt(Minimum=1, Maximum=5)
    Use Spawn Probability=false
    UseLoopCountLimit=false
   )
  ]
  particleSpawnScript [
   &InitializeParticle(
    Color Mode=::NewEnumerator0
    Lifetime Mode=::NewEnumerator1
    Mesh Scale Mode=::NewEnumerator1
    Mesh Uniform Scale Max=100
    Mesh Uniform Scale Min=70
    UsePositionOffset=false
    Write Material Random=true
   )
   &CylinderLocation(
    Cylinder Height=0
    Cylinder Height Midpoint=0
    Cylinder Origin:NiagaraPosition=$User.Stop Location
    Cylinder Radius=500
    MaskSpawnGroup=false
    Offset=<0 250 0>
    Offset Coordinate Space:ENiagaraCoordinateSpace=::Local
    Offset Mode:ENiagara_OffsetMode=::NewEnumerator0
    Override Seed=false
    UseOwnerScale=true
   )
   &InitialMeshOrientation(
    Mesh Orientation Mode=::NewEnumerator7
    Rotation=&RandomRangeVector(Minimum=<0 0 0>, Maximum=<0 0 1>)
    Use Rotation Vector=true
   )
   &AddVelocityInCone(
    Cone Axis=&self:Reflection_Vector(
     Normalized Normal:Vector3f=&NormalizeVector(Vector To Normalize=$User.Stop Normal, Safe Normalize=true, Safe Fallback Threshold=0, Safe Fallback Vector=<0 0 0.001>)
     Stop Location:Vector3f=Linked() {linkedVariable {name=User.Stop Location, type=NiagaraPosition}}
    )
    Cone Axis Coordinate Space=::World
    Module.Use Velocity Falloff On Cone Axis=false
    Velocity Strength=50
   )
  ]
  particleUpdateScript [
   &SolveForcesAndVelocity(Clamp Velocity=false, Limit Acceleration=false)
   &ParticleState()
   &DynamicMaterialParameters(
    Index 0 Param 1=&FloatFromCurve(FloatCurve=DataInterface() {propertyValues {Curve {keys [<at -0.0046 1>,<at 1.0011 4>]}}})
    Index 0 Param 2=&FloatFromCurve(FloatCurve=DataInterface() {propertyValues {Curve {keys [<at -0.0046 0.2>,<at 1.0011 1.3>]}}})
    Index 0 Param 3=&FloatFromCurve(FloatCurve=DataInterface() {propertyValues {Curve {keys [<at -0.0046 1>,<at 1.0011 0.5>]}}})
    Index 2 Param 1=&FloatFromCurve(FloatCurve=DataInterface() {propertyValues {Curve {keys [<at -0.0046 0.4>,<at 1.0011 1.5>]}}})
    Index 2 Param 2=&FloatFromCurve(FloatCurve=DataInterface() {propertyValues {Curve {keys [<at -0.0046 0.1>,<at 0.2854 2.6>,<at 1.0011 0.3>]}}})
    Index 2 Param 3=0.8
    Index 2 Param 4=&FloatFromCurve(FloatCurve=DataInterface() {propertyValues {Curve {keys [
         {arriveTangent=-1.9991, interpMode=RCIM_Cubic, time=0, value=1}
         {arriveTangent=-2.7143, interpMode=RCIM_Cubic, leaveTangent=-2.7143, tangentMode=RCTM_User, time=0.2095, value=0.6901}
         <at 1 0 RCIM_Cubic>
        ]}}})
    Index 3 Param 1=$User.WaterBodyIndex
    Index 3 Param 2=$User.WaterDepth
    Index 3 Param 3=$User.WaveAttenuationFactor
    Module.Param0WriteEnabled=true
    Module.Param0WriteEnabled2=true
    Module.Param0WriteEnabled3=true
    Module.Param1WriteEnabled=true
    Module.Param1WriteEnabled2=true
    Module.Param1WriteEnabled3=true
    Module.Param2WriteEnabled=true
    Module.Param2WriteEnabled2=true
    Module.Param2WriteEnabled3=true
    Module.Param3WriteEnabled=true
    Module.Param3WriteEnabled2=true
    Module.Param3WriteEnabled3=true
    Write Parameter Index 2=true
    Write Parameter Index 3=true
   ) {module=DynamicMaterialParameters001}
  ]
  rendererClasses [&NiagaraMeshRendererProperties]
  renderers [{rendererClass=&NiagaraMeshRendererProperties, rendererIndex=0}]
 }
 Mist {
  emitterUpdateScript [
   &EmitterState(Life Cycle Mode=::NewEnumerator1, Loop Duration=0.1, UseLoopDelay=false)
   &SpawnBurst_Instantaneous(
    Spawn Count=&RandomRangeInt(Minimum=0, Maximum=1)
    Use Spawn Probability=false
    UseLoopCountLimit=false
   )
  ]
  particleSpawnScript [
   &InitializeParticle(
    Lifetime Max=1.2
    Lifetime Min=0.8
    Lifetime Mode=::NewEnumerator1
    Sprite Rotation Mode=::NewEnumerator2
    Sprite Size Mode=::NewEnumerator1
    Uniform Sprite Size Max=280
    Uniform Sprite Size Min=250
    UsePositionOffset=false
    Write Material Random=true
   )
   &CylinderLocation(
    Cylinder Height=0
    Cylinder Height Midpoint=0
    Cylinder Origin:NiagaraPosition=$User.Stop Location
    Cylinder Radius=50
    MaskSpawnGroup=false
    Override Seed=false
    UseOwnerScale=true
   )
   &AddVelocityInCone(
    Cone Angle=30
    Cone Axis=&self:Reflection_Vector(
     Normalized Normal:Vector3f=&NormalizeVector(Vector To Normalize=$User.Stop Normal, Safe Normalize=true, Safe Fallback Threshold=0, Safe Fallback Vector=<0 0 0.001>)
     Stop Location:Vector3f=Linked() {linkedVariable {name=User.Stop Location, type=NiagaraPosition}}
    )
    Cone Axis Coordinate Space=::World
    Module.Use Velocity Falloff On Cone Axis=false
    Velocity Strength=800
   )
   &AddVelocity(Coordinate Space=::World, Velocity=&RandomRangeVector(Minimum=<0 0 150>, Maximum=<0 0 750>))
   &DynamicMaterialParameters(
    Index 0 Param 1=&RandomRangeFloat(Minimum=1, Maximum=2)
    Index 0 Param 2=1
    Index 0 Param 3=4
    Index 0 Param 4=0.01
    Module.Param0WriteEnabled=true
    Module.Param1WriteEnabled=true
    Module.Param2WriteEnabled=true
    Module.Param3WriteEnabled=true
   )
  ]
  particleUpdateScript [
   &GravityForce(Gravity=<0 0 -350>)
   &Drag(Use Linear Drag=true, Use Rotational Drag=false)
   &SolveForcesAndVelocity(Clamp Velocity=false, Limit Acceleration=false)
   &ParticleState()
   &SubUVAnimation(
    SubUV Animation Mode:ENiagara_SubUVLookupMode=::NewEnumerator1
    SubUV Curve=DataInterface() {propertyValues {Curve {keys [{interpMode=RCIM_Cubic, leaveTangent=2.3356, tangentMode=RCTM_Break, time=0, value=0}, <at 1 0.9>]}}}
   )
   &ScaleSpriteSize(
    Uniform Curve Scale=2
    Uniform Curve Sprite Scale=DataInterface() {propertyValues {Curve {keys [{tangentMode=RCTM_User, time=0, value=0.5}, {tangentMode=RCTM_User, time=1, value=2}]}}}
   )
   &ScaleColor(
    Scale Alpha=&FloatFromCurve(FloatCurve=DataInterface() {propertyValues {Curve {keys [
         {arriveTangent=26.1673, interpMode=RCIM_Cubic, leaveTangent=26.1673, tangentMode=RCTM_User, time=0, value=0}
         {arriveTangent=0.002, interpMode=RCIM_Cubic, leaveTangent=0.002, time=0.1, value=2}
         <at 1.0011 0.002 RCIM_Cubic>
        ]}}})
    Scale RGB=&MakeVectorFromLinearColorRGB(Color=&ColorFromCurve(ColorCurve=DataInterface() {propertyValues {
        AlphaCurve {keys []}
        BlueCurve {keys [{arriveTangent=-92.86, interpMode=RCIM_Cubic, leaveTangent=-92.86, tangentMode=RCTM_User, time=0, value=10}, <at 0.25 1.5 RCIM_Cubic>]}
        GreenCurve {keys [{arriveTangent=7.9315, interpMode=RCIM_Cubic, leaveTangent=7.9315, tangentMode=RCTM_User, time=0, value=0.3013}, <at 0.25 1.0432 RCIM_Cubic>]}
        RedCurve {keys [{arriveTangent=-50.8375, interpMode=RCIM_Cubic, leaveTangent=-50.8375, tangentMode=RCTM_User, time=0, value=5}, <at 0.25 0.612 RCIM_Cubic>]}
       }}))
    ScaleA=true
    ScaleRGB=true
   )
  ]
  rendererClasses [&NiagaraSpriteRendererProperties]
  renderers [{rendererClass=&NiagaraSpriteRendererProperties}]
 }
 Water_Foam {
  emitterUpdateScript [
   &EmitterState(Life Cycle Mode=::NewEnumerator1, Loop Delay=0.05, Loop Duration=0.1, UseLoopDelay=false)
   &SpawnBurst_Instantaneous(Use Spawn Probability=false, UseLoopCountLimit=false)
   &SpawnBurst_Instantaneous(Spawn Group=1, Use Spawn Probability=false, UseLoopCountLimit=false) {module=SpawnBurst_Instantaneous001}
  ]
  particleSpawnScript [
   &InitializeParticle(
    Lifetime=&MaskFloatBySpawnGroup(
     Masked Float=&RandomRangeFloat(Minimum=0.5, Maximum=1)
     Passthrough Float=&RandomRangeFloat(Minimum=1, Maximum=1.5)
     Spawn Group Mask=1
    )
    Mesh Scale Mode=::NewEnumerator1
    Mesh Uniform Scale Max=400
    Mesh Uniform Scale Min=200
    Sprite Rotation Mode=::NewEnumerator2
    UsePositionOffset=false
    Write Material Random=true
   )
   &SystemLocation(
    MaskSpawnGroup=false
    Offset=&TransformPosition(Position=Linked() {linkedVariable {name=User.Stop Location, type=NiagaraPosition}}, Source Space=::World, Destination Space=::Local)
   )
   &InitialMeshOrientation(
    Mesh Orientation Mode=::NewEnumerator7
    Rotation=&MaskVectorBySpawnGroup(
     Masked Vector=&RandomRangeVector(Minimum=<0 0 0>, Maximum=<0 0 1>)
     Passthrough Vector=&RandomRangeVector(Minimum=<0 0 0>, Maximum=<0 0 1>)
    )
    Use Rotation Vector=true
   )
   &AddVelocityInCone(
    Cone Axis=&self:Reflection_Vector(
     Normalized Normal:Vector3f=&NormalizeVector(Vector To Normalize=$User.Stop Normal, Safe Normalize=true, Safe Fallback Threshold=0, Safe Fallback Vector=<0 0 0.001>)
     Stop Location:Vector3f=Linked() {linkedVariable {name=User.Stop Location, type=NiagaraPosition}}
    )
    Cone Axis Coordinate Space=::World
    Module.Use Velocity Falloff On Cone Axis=false
    Velocity Strength=250
   )
   &ConstrainPositionToPlane(Plane Normal:Vector3f=$User.Stop Normal, Plane Position:NiagaraPosition=$User.Stop Location)
  ]
  particleUpdateScript [
   &SolveForcesAndVelocity(Clamp Velocity=false, Limit Acceleration=false)
   &ParticleState()
   &DynamicMaterialParameters(
    Index 0 Param 1=&FloatFromCurve(FloatCurve=DataInterface() {propertyValues {Curve {keys [<at -0.0046 1>,<at 1.0011 4>]}}})
    Index 0 Param 2=&FloatFromCurve(FloatCurve=DataInterface() {propertyValues {Curve {keys [{arriveTangent=0.6218, interpMode=RCIM_Cubic, leaveTangent=0.6218, tangentMode=RCTM_User, time=-0.0046, value=0.5}, <at 1.0011 1 RCIM_Cubic>]}}})
    Index 0 Param 3=&FloatFromCurve(FloatCurve=DataInterface() {propertyValues {Curve {keys [<at -0.0046 1>,<at 1.0011 0.5>]}}})
    Index 1 Param 3=&MaskFloatBySpawnGroup(Masked Float=500, Passthrough Float=800)
    Index 2 Param 1=&FloatFromCurve(FloatCurve=DataInterface() {propertyValues {Curve {keys [<at 0 0 RCIM_Cubic>, {arriveTangent=3.8777, interpMode=RCIM_Cubic, leaveTangent=3.8777, tangentMode=RCTM_User, time=1.0011, value=1.5}]}}})
    Index 2 Param 2=&FloatFromCurve(FloatCurve=DataInterface() {propertyValues {Curve {keys [
         {arriveTangent=-4.0775, interpMode=RCIM_Cubic, leaveTangent=-4.0775, tangentMode=RCTM_User, time=-0.0046, value=2}
         <at 1.0011 0.4 RCIM_Cubic>
        ]}}})
    Index 2 Param 3=&MaskFloatBySpawnGroup(Masked Float=0.5, Passthrough Float=1)
    Index 2 Param 4=&FloatFromCurve(FloatCurve=DataInterface() {propertyValues {Curve {keys [
         {arriveTangent=-1.9991, interpMode=RCIM_Cubic, time=0, value=1}
         {arriveTangent=-2.7143, interpMode=RCIM_Cubic, leaveTangent=-2.7143, tangentMode=RCTM_User, time=0.2095, value=0.6901}
         <at 1 0 RCIM_Cubic>
        ]}}})
    Index 3 Param 1=$User.WaterBodyIndex
    Index 3 Param 2=$User.WaterDepth
    Index 3 Param 3=$User.WaveAttenuationFactor
    Module.Param0WriteEnabled=true
    Module.Param0WriteEnabled1=false
    Module.Param0WriteEnabled2=true
    Module.Param0WriteEnabled3=true
    Module.Param1WriteEnabled=true
    Module.Param1WriteEnabled1=false
    Module.Param1WriteEnabled2=true
    Module.Param1WriteEnabled3=true
    Module.Param2WriteEnabled=true
    Module.Param2WriteEnabled1=true
    Module.Param2WriteEnabled2=true
    Module.Param2WriteEnabled3=true
    Module.Param3WriteEnabled=true
    Module.Param3WriteEnabled1=false
    Module.Param3WriteEnabled2=true
    Module.Param3WriteEnabled3=true
    Write Parameter Index 1=true
    Write Parameter Index 2=true
    Write Parameter Index 3=true
   )
   &ScaleMeshSize(Scale Factor=&VectorFromFloat(Value=&MaskFloatBySpawnGroup(
      Masked Float=&FloatFromCurve(FloatCurve=DataInterface() {propertyValues {Curve {keys [
           {arriveTangent=26.4511, interpMode=RCIM_Cubic, leaveTangent=26.451, tangentMode=RCTM_User, time=-0.0046, value=0.6073}
           {arriveTangent=3.3733, interpMode=RCIM_Cubic, leaveTangent=3.3733, time=0.1268, value=2.5016}
           <at 1.0011 4 RCIM_Cubic>
          ]}}})
      Passthrough Float=&FloatFromCurve(FloatCurve=DataInterface() {propertyValues {Curve {keys [
           {arriveTangent=26.4511, interpMode=RCIM_Cubic, leaveTangent=26.451, tangentMode=RCTM_User, time=-0.0046, value=0.6073}
           {arriveTangent=3.3733, interpMode=RCIM_Cubic, leaveTangent=3.3733, time=0.1268, value=2.5016}
           <at 1.0011 4 RCIM_Cubic>
          ]}}})
     )))
  ]
  rendererClasses [&NiagaraMeshRendererProperties]
  renderers [{rendererClass=&NiagaraMeshRendererProperties, rendererIndex=0}]
 }
}
summary {
 systemName=NS_Spire_BifucateBeam_Impact_Water
 emitters [
  {emitterName=Arcing_Water_Splashes, rendererClasses [&NiagaraMeshRendererProperties]}
  {emitterName=Water_Foam, rendererClasses [&NiagaraMeshRendererProperties]}
  {emitterName=Foam_Details, rendererClasses [&NiagaraMeshRendererProperties]}
  {emitterName=Drop_Details, rendererClasses [&NiagaraSpriteRendererProperties]}
  {emitterName=ChevronShapes, rendererClasses [&NiagaraMeshRendererProperties]}
  {emitterName=Mist, rendererClasses [&NiagaraSpriteRendererProperties]}
 ]
 userVariables [
  User.Stop Location {type=NiagaraPosition, description="", defaultValue {}}
  User.Stop Normal {type=Vector3f, description="", defaultValue {struct=Vector3f, value=<0 0 1>}}
  User.WaterBodyIndex {type=NiagaraFloat, defaultValue=0, description=""}
  User.WaterDepth {type=NiagaraFloat, defaultValue=0, description=""}
  User.WaterPlaneLocation {type=Vector3f, description="", defaultValue {struct=Vector3f, value=<0 0 0>}}
  User.WaterPlaneNormal {type=Vector3f, description="", defaultValue {struct=Vector3f, value=<0 0 0>}}
  User.WaveAttenuationFactor {type=NiagaraFloat, defaultValue=0, description=""}
  User.WaveTime {type=NiagaraFloat, defaultValue=0, description=""}
 ]
}
