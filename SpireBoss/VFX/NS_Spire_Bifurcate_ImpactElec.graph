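# Niagara system for an electrical impact effect, named for the Spire boss's Bifurcate.
# It contains three emitters, Lightning, Smoke and Light, with no renderers listed for any of them.
# User primary and secondary color parameters expose tinting, both defaulting to reddish hues.
# The system script has two variable-setting modules alongside SystemState, and its dependencies include curve, waveform, multiply and color-building dynamic inputs.
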NiagaraSystem /SpireBoss/VFX/NS_Spire_Bifurcate_ImpactElec
dependencies {
 usedDataInterfaces [&NiagaraDataInterfaceCurve]
 usedDynamicInputs [
  &Multiply_Float
  &FloatFromCurve
  &MakeLinearColorFromVectorAndFloat
  &VectorFromFloat
  &Waveform
 ]
 usedModules [
  &SystemState
  &self:SystemSpawnScript.SystemScriptSource.SystemScriptGraph.NiagaraNodeAssignment_0.SetVariables_#1
  &self:SystemSpawnScript.SystemScriptSource.SystemScriptGraph.NiagaraNodeAssignment_1.SetVariables_#2
 ]
}
emitters {Light {rendererClasses [], renderers []}, Lightning {rendererClasses [], renderers []}, Smoke {rendererClasses [], renderers []}}
summary {
 systemName=NS_Spire_Bifurcate_ImpactElec
 emitters [{emitterName=Lightning, rendererClasses []}, {emitterName=Smoke, rendererClasses []}, {emitterName=Light, rendererClasses []}]
 userVariables [
  User.PrimaryColor {type=LinearColor, description="", defaultValue {struct=LinearColor, value=<rgba 5 0.2882 0.6068 1>}}
  User.SecondaryColor {type=LinearColor, description="", defaultValue {struct=LinearColor, value=<rgba 0.7101 0.0959 0.1594 5>}}
 ]
}
